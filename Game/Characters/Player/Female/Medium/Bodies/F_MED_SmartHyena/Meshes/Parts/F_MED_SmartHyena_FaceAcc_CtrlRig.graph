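# Control Rig for the SmartHyena female player character's face-accessory part, procedurally posing back, left and right hair bone chains against body motion.
# Setup stores each chain's controls and poses in variables.
# Forward solve parent-constrains control nulls to weighted head, neck and spine bones, runs B-spline IK through the controls, then distributes twist.
# Upper-arm spherical pose readers weight corrective offsets on the side chains; the two baseFix constraints are unconnected.
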ControlRigBlueprint /Game/Characters/Player/Female/Medium/Bodies/F_MED_SmartHyena/Meshes/Parts/F_MED_SmartHyena_FaceAcc_CtrlRig
graphs [
 ContainedGraph {nodes [
   n1 = RigVMFunctionEntryNode()
   SequenceAggregate = &RigVMFunction_Sequence(ExecuteContext=@n1.ExecuteContext)
   SequenceAggregate_1 = &RigVMFunction_Sequence(ExecuteContext=@SequenceAggregate.B)
   SequenceAggregate_2 = &RigVMFunction_Sequence(ExecuteContext=@SequenceAggregate_1.B)
   RigVMFunctionReturnNode(A=@SequenceAggregate.A, B=@SequenceAggregate_1.A, C=@SequenceAggregate_2.A, D=@SequenceAggregate_2.B)
  ]}
 ContainedGraph {nodes [
   n1 = RigVMFunctionEntryNode()
   SequenceAggregate = &RigVMFunction_Sequence(ExecuteContext=@n1.ExecuteContext)
   SequenceAggregate_1 = &RigVMFunction_Sequence(ExecuteContext=@SequenceAggregate.B)
   SequenceAggregate_2 = &RigVMFunction_Sequence(ExecuteContext=@SequenceAggregate_1.B)
   RigVMFunctionReturnNode(A=@SequenceAggregate.A, B=@SequenceAggregate_1.A, C=@SequenceAggregate_2.A, D=@SequenceAggregate_2.B)
  ]}
 RigVMFunctionLibrary {links [], nodes []}
 RigVMModel {nodes [
   RigUnit_BeginExecution = &RigUnit_BeginExecution()
   SequenceAggregate = RigVMAggregateNode(ExecuteContext=@RigUnit_BeginExecution.ExecutePin)
   &RigUnit_ParentConstraint(
    AdvancedSettings=(RotationOrderForFilter=XZY)
    Child=(Type=Null,Name="dyn_Hair_Back_01_Null")
    Filter=(TranslationFilter=(bX=true,bY=true,bZ=true),RotationFilter=(bX=false,bY=false,bZ=false),ScaleFilter=(bX=True,bY=True,bZ=True))
    Parents=((Item=(Type=Bone,Name="head"),Weight=0.9),(Item=(Type=Bone,Name="spine_05"),Weight=0.1))
    Weight=1
    ExecutePin=@SequenceAggregate.A
   )
   > &RigUnit_ParentConstraint(
    AdvancedSettings=(RotationOrderForFilter=XZY)
    Child=(Type=Null,Name="dyn_Hair_Back_02_Null")
    Filter=(TranslationFilter=(bX=True,bY=True,bZ=True),RotationFilter=(bX=false,bY=false,bZ=false),ScaleFilter=(bX=True,bY=True,bZ=True))
    Parents=((Item=(Type=Bone,Name="neck_01"),Weight=0.8),(Item=(Type=Bone,Name="spine_05"),Weight=0.2))
    Weight=1
   )
   > &RigUnit_ParentConstraint(
    AdvancedSettings=(RotationOrderForFilter=XZY)
    Child=(Type=Null,Name="dyn_Hair_Back_03_Null")
    Filter=(TranslationFilter=(bX=True,bY=True,bZ=True),RotationFilter=(bX=false,bY=false,bZ=false),ScaleFilter=(bX=True,bY=True,bZ=True))
    Parents=((Item=(Type=Bone,Name="neck_02"),Weight=0.8),(Item=(Type=Bone,Name="spine_05"),Weight=0.2))
    Weight=1
   )
   > &RigUnit_ParentConstraint(
    AdvancedSettings=(RotationOrderForFilter=XZY)
    Child=(Type=Null,Name="dyn_Hair_Back_04_Null")
    Filter=(TranslationFilter=(bX=True,bY=True,bZ=True),RotationFilter=(bX=false,bY=false,bZ=false),ScaleFilter=(bX=True,bY=True,bZ=True))
    Parents=((Item=(Type=Bone,Name="clavicle_r"),Weight=0),(Item=(Type=Bone,Name="spine_04"),Weight=1),(Item=(Type=Bone,Name="clavicle_l"),Weight=0))
    Weight=1
   )
   > &RigUnit_ParentConstraint(
    AdvancedSettings=(RotationOrderForFilter=XZY)
    Child=(Type=Null,Name="dyn_Hair_Back_05_Null")
    Filter=(TranslationFilter=(bX=True,bY=True,bZ=True),RotationFilter=(bX=false,bY=false,bZ=false),ScaleFilter=(bX=True,bY=True,bZ=True))
    Parents=((Item=(Type=Bone,Name="clavicle_l"),Weight=0),(Item=(Type=Bone,Name="spine_03"),Weight=1),(Item=(Type=Bone,Name="clavicle_r"),Weight=0))
    Weight=1
   )
   RerouteNode_2 = RigVMDispatchNode(Values=((Type=Bone,Name="dyn_Hair_Back_01"),((),Type=Bone,Name="dyn_Hair_Back_02"),((),Type=Bone,Name="dyn_Hair_Back_03"),((),Type=Bone,Name="dyn_Hair_Back_04"),((),Type=Bone,Name="dyn_Hair_Back_05"),((),Type=Bone,Name="None"))) {template="DISPATCH_RigVMDispatch_ArrayMake(in Values,out Array)"}
   > SplineIK_2 = RigVMFunctionReferenceNode(
    Primary Axis=(X=1,Y=-0.1,Z=0)
    Samples Per Segment=1
    Secondary Spline Direction=(X=0,Y=0,Z=0)
    Spline Mode=BSpline
    Stretch=false
    Up Axis=(X=0,Y=0,Z=0)
    debug=false
    Bones=@RerouteNode_2.Array
    Controls=RigVMVariableNode() {variable=dyn_Hair_Back_ctrls}.Value
   )
   > TwistBones = RigVMFunctionReferenceNode(
    Axis to Twist=(bX=true,bY=true,bZ=true)
    Bones to Twist=((Type=Bone,Name="dyn_Hair_Back_02"),(Type=Bone,Name="dyn_Hair_Back_03"))
    End=(Type=Bone,Name="neck_01")
    Start=(Type=Bone,Name="spine_05")
   )
   > TwistBones_3 = RigVMFunctionReferenceNode(
    Axis to Twist=(bX=true,bY=true,bZ=true)
    Bones to Twist=((Type=Bone,Name="dyn_Hair_Back_04"),(Type=Bone,Name="dyn_Hair_Back_05"),(Type=None,Name="None"))
    End=(Type=Bone,Name="spine_05")
    Start=(Type=Bone,Name="spine_04")
   )
   &RigUnit_ParentConstraint(
    AdvancedSettings=(RotationOrderForFilter=XZY)
    Child=(Type=Null,Name="dyn_Hair_l_base_Null")
    Filter=(TranslationFilter=(bX=true,bY=true,bZ=true),RotationFilter=(bX=false,bY=false,bZ=false),ScaleFilter=(bX=True,bY=True,bZ=True))
    Parents=((Item=(Type=Bone,Name="neck_01"),Weight=0.5),(Item=(Type=Bone,Name="spine_05"),Weight=0.5))
    Weight=1
    ExecutePin=@SequenceAggregate.B
   )
   > &RigUnit_ParentConstraint(
    AdvancedSettings=(RotationOrderForFilter=XZY)
    Child=(Type=Null,Name="dyn_Hair_l_01_Null")
    Filter=(TranslationFilter=(bX=True,bY=True,bZ=True),RotationFilter=(bX=false,bY=false,bZ=false),ScaleFilter=(bX=True,bY=True,bZ=True))
    Parents=((Item=(Type=Bone,Name="neck_01"),Weight=0.5),(Item=(Type=Bone,Name="spine_05"),Weight=0.5))
    Weight=1
   )
   > &RigUnit_ParentConstraint(
    AdvancedSettings=(RotationOrderForFilter=XZY)
    Child=(Type=Null,Name="dyn_Hair_l_02_Null")
    Filter=(TranslationFilter=(bX=True,bY=True,bZ=True),RotationFilter=(bX=false,bY=false,bZ=false),ScaleFilter=(bX=True,bY=True,bZ=True))
    Parents=((Item=(Type=Bone,Name="neck_02"),Weight=0.3),(Item=(Type=Bone,Name="spine_05"),Weight=0.7))
    Weight=1
   )
   > &RigUnit_ParentConstraint(
    AdvancedSettings=(RotationOrderForFilter=XZY)
    Child=(Type=Null,Name="dyn_Hair_l_03_Null")
    Filter=(TranslationFilter=(bX=True,bY=True,bZ=True),RotationFilter=(bX=false,bY=false,bZ=false),ScaleFilter=(bX=True,bY=True,bZ=True))
    Parents=((Item=(Type=Bone,Name="clavicle_r"),Weight=0),(Item=(Type=Bone,Name="spine_04"),Weight=1),(Item=(Type=Bone,Name="clavicle_l"),Weight=0))
    Weight=1
   )
   > &RigUnit_ParentConstraint(
    AdvancedSettings=(RotationOrderForFilter=XZY)
    Child=(Type=Null,Name="dyn_Hair_l_04_Null")
    Filter=(TranslationFilter=(bX=True,bY=True,bZ=True),RotationFilter=(bX=false,bY=false,bZ=false),ScaleFilter=(bX=True,bY=True,bZ=True))
    Parents=((Item=(Type=Bone,Name="clavicle_l"),Weight=0),(Item=(Type=Bone,Name="spine_03"),Weight=1),(Item=(Type=Bone,Name="clavicle_r"),Weight=0))
    Weight=1
   )
   RerouteNode_2_1 = RigVMDispatchNode(Values=((Type=Bone,Name="dyn_Hair_l_base"),((),Type=Bone,Name="dyn_Hair_l_01"),((),Type=Bone,Name="dyn_Hair_l_02"),((),Type=Bone,Name="dyn_Hair_l_03"),((),Type=Bone,Name="dyn_Hair_l_04"),((),Type=Bone,Name="None"))) {template="DISPATCH_RigVMDispatch_ArrayMake(in Values,out Array)"}
   > SplineIK_2_1 = RigVMFunctionReferenceNode(
    Primary Axis=(X=0,Y=0,Z=0)
    Samples Per Segment=1
    Secondary Spline Direction=(X=0,Y=0,Z=0)
    Spline Mode=BSpline
    Stretch=false
    Up Axis=(X=0,Y=0,Z=0)
    debug=false
    Bones=@RerouteNode_2_1.Array
    Controls=RigVMVariableNode() {variable=dyn_Hair_L_ctrls}.Value
   )
   > TwistBones_1 = RigVMFunctionReferenceNode(
    Axis to Twist=(bX=true,bY=true,bZ=true)
    Bones to Twist=((Type=Bone,Name="dyn_Hair_l_02"),(Type=Bone,Name="dyn_Hair_l_04"))
    End=(Type=Bone,Name="neck_02")
    Start=(Type=Bone,Name="spine_04")
   )
   > n1 = &RigUnit_SphericalPoseReader(
    ActiveRegionScaleFactors=(PositiveWidth=2.5,NegativeWidth=3,PositiveHeight=3,NegativeHeight=3)
    ActiveRegionSize=0.1
    Debug=(bDrawDebug=false,DebugScale=25,DebugSegments=20,DebugThickness=0.25)
    DriverAxis=(X=1,Y=0,Z=0)
    DriverItem=(Type=Bone,Name="upperarm_l")
    FalloffRegionScaleFactors=(PositiveWidth=1,NegativeWidth=1,PositiveHeight=1,NegativeHeight=1)
    FalloffSize=0.55
    OptionalParentItem=(Type=Bone)
    RotationOffset=(X=45,Y=30,Z=0)
   )
   > &RigUnit_ModifyTransforms(
    ItemToModify="((Item=(Type=Bone,Name=\"dyn_Hair_l_base\"),Transform=(Rotation=(X=0,Y=0,Z=0,W=1),Translation=(X=0,Y=0,Z=-5),Scale3D=(X=1,Y=1,Z=1))),(Item=(Type=Bone,Name=\"dyn_Hair_l_01\"),Transform=(Rotation=(X=0,Y=0,Z=0,W=1),Translation=(X=0,Y=0,Z=3),Scale3D=(X=1,Y=1,Z=1)) ...(truncated)"
    ItemToModify.1.Transform.Rotation=&RigVMFunction_MathQuaternionFromEuler(Euler=(X=0,Y=5,Z=0)).Result
    Weight=@n1.OutputParam
   )
   &RigUnit_ParentConstraint(
    AdvancedSettings=(RotationOrderForFilter=XZY)
    Child=(Type=Null,Name="dyn_Hair_r_base_Null")
    Filter=(TranslationFilter=(bX=true,bY=true,bZ=true),RotationFilter=(bX=true,bY=true,bZ=true),ScaleFilter=(bX=True,bY=True,bZ=True))
    Parents=((Item=(Type=Bone,Name="neck_01"),Weight=0.5),(Item=(Type=Bone,Name="spine_05"),Weight=0.5))
    Weight=1
    ExecutePin=@SequenceAggregate.C
   )
   > &RigUnit_ParentConstraint(
    AdvancedSettings=(RotationOrderForFilter=XZY)
    Child=(Type=Null,Name="dyn_Hair_r_01_Null")
    Filter=(TranslationFilter=(bX=True,bY=True,bZ=True),RotationFilter=(bX=false,bY=false,bZ=false),ScaleFilter=(bX=True,bY=True,bZ=True))
    Parents=((Item=(Type=Bone,Name="neck_01"),Weight=0.5),(Item=(Type=Bone,Name="spine_05"),Weight=0.5))
    Weight=1
   )
   > &RigUnit_ParentConstraint(
    AdvancedSettings=(RotationOrderForFilter=XZY)
    Child=(Type=Null,Name="dyn_Hair_r_02_Null")
    Filter=(TranslationFilter=(bX=True,bY=True,bZ=True),RotationFilter=(bX=false,bY=false,bZ=false),ScaleFilter=(bX=True,bY=True,bZ=True))
    Parents=((Item=(Type=Bone,Name="neck_02"),Weight=0.3),(Item=(Type=Bone,Name="spine_05"),Weight=0.7))
    Weight=1
   )
   > &RigUnit_ParentConstraint(
    AdvancedSettings=(RotationOrderForFilter=XZY)
    Child=(Type=Null,Name="dyn_Hair_r_03_Null")
    Filter=(TranslationFilter=(bX=True,bY=True,bZ=True),RotationFilter=(bX=false,bY=false,bZ=false),ScaleFilter=(bX=True,bY=True,bZ=True))
    Parents=((Item=(Type=Bone,Name="clavicle_r"),Weight=0),(Item=(Type=Bone,Name="spine_04"),Weight=1),(Item=(Type=Bone,Name="clavicle_l"),Weight=0))
    Weight=1
   )
   > &RigUnit_ParentConstraint(
    AdvancedSettings=(RotationOrderForFilter=XZY)
    Child=(Type=Null,Name="dyn_Hair_r_04_Null")
    Filter=(TranslationFilter=(bX=True,bY=True,bZ=True),RotationFilter=(bX=false,bY=false,bZ=false),ScaleFilter=(bX=True,bY=True,bZ=True))
    Parents=((Item=(Type=Bone,Name="clavicle_l"),Weight=0),(Item=(Type=Bone,Name="spine_03"),Weight=1),(Item=(Type=Bone,Name="clavicle_r"),Weight=0))
    Weight=1
   )
   RerouteNode_2_2 = RigVMDispatchNode(Values=((Type=Bone,Name="dyn_Hair_r_base"),((),Type=Bone,Name="dyn_Hair_r_01"),((),Type=Bone,Name="dyn_Hair_r_02"),((),Type=Bone,Name="dyn_Hair_r_03"),((),Type=Bone,Name="dyn_Hair_r_04"),((),Type=Bone,Name="None"))) {template="DISPATCH_RigVMDispatch_ArrayMake(in Values,out Array)"}
   > SplineIK_2_2 = RigVMFunctionReferenceNode(
    Primary Axis=(X=0,Y=0,Z=0)
    Samples Per Segment=1
    Secondary Spline Direction=(X=0,Y=0,Z=0)
    Spline Mode=BSpline
    Stretch=false
    Up Axis=(X=0,Y=0,Z=0)
    debug=false
    Bones=@RerouteNode_2_2.Array
    Controls=RigVMVariableNode() {variable=dyn_Hair_R_ctrls}.Value
   )
   > TwistBones_2 = RigVMFunctionReferenceNode(
    Axis to Twist=(bX=true,bY=true,bZ=true)
    Bones to Twist=((Type=Bone,Name="dyn_Hair_r_02"),(Type=Bone,Name="dyn_Hair_r_04"))
    End=(Type=Bone,Name="neck_02")
    Start=(Type=Bone,Name="spine_04")
   )
   > n2 = &RigUnit_SphericalPoseReader(
    ActiveRegionScaleFactors=(PositiveWidth=2.5,NegativeWidth=3,PositiveHeight=3,NegativeHeight=3)
    ActiveRegionSize=0.1
    Debug=(bDrawDebug=false,DebugScale=25,DebugSegments=20,DebugThickness=0.25)
    DriverAxis=(X=-1,Y=0,Z=0)
    DriverItem=(Type=Bone,Name="upperarm_r")
    FalloffRegionScaleFactors=(PositiveWidth=1,NegativeWidth=1,PositiveHeight=1,NegativeHeight=1)
    FalloffSize=0.6
    OptionalParentItem=(Type=Bone)
    RotationOffset=(X=-60,Y=210,Z=0)
   )
   n3 = &RigVMFunction_MathQuaternionFromEuler(Euler=(X=0,Y=-5,Z=15))
   > &RigUnit_ModifyTransforms(
    ItemToModify="((Item=(Type=Bone,Name=\"dyn_Hair_r_base\"),Transform=(Rotation=(X=0,Y=0,Z=0,W=1),Translation=(X=0,Y=0,Z=6),Scale3D=(X=1,Y=1,Z=1))),(Item=(Type=Bone,Name=\"dyn_Hair_r_03\"),Transform=(Rotation=(X=0,Y=0,Z=0,W=1),Translation=(X=0,Y=0,Z=2),Scale3D=(X=1,Y=1,Z=1))) ...(truncated)"
    ItemToModify.1.Transform.Rotation=@n3.Result
    ItemToModify.2.Transform.Rotation=@n3.Result
    ItemToModify.0.Transform.Rotation=&RigVMFunction_MathQuaternionFromEuler(Euler=(X=0,Y=5,Z=0)).Result
    Weight=@n2.OutputParam
   )
   &RigUnit_ParentConstraint(
    AdvancedSettings=(RotationOrderForFilter=XZY)
    Child=(Type=Bone,Name="dyn_Hair_r_baseFix")
    ExecutePin=()
    Filter=(TranslationFilter=(bX=true,bY=true,bZ=true),RotationFilter=(bX=true,bY=true,bZ=true),ScaleFilter=(bX=True,bY=True,bZ=True))
    Parents=((Item=(Type=Bone,Name="head"),Weight=0.5),(Item=(Type=Bone,Name="neck_02"),Weight=0.3),(Item=(Type=Bone,Name="spine_05"),Weight=0.2))
    Weight=1
   )
   &RigUnit_ParentConstraint(
    AdvancedSettings=(RotationOrderForFilter=XZY)
    Child=(Type=Bone,Name="dyn_Hair_l_baseFix")
    ExecutePin=()
    Filter=(TranslationFilter=(bX=true,bY=true,bZ=true),RotationFilter=(bX=true,bY=true,bZ=true),ScaleFilter=(bX=True,bY=True,bZ=True))
    Parents=((Item=(Type=Bone,Name="head"),Weight=0.5),(Item=(Type=Bone,Name="neck_02"),Weight=0.3),(Item=(Type=Bone,Name="spine_05"),Weight=0.2))
    Weight=1
   )
  ]}
 RigVMModel Rig Graph {nodes [
   n1 = &RigUnit_PrepareForExecution()
   SequenceAggregate = RigVMAggregateNode(ExecuteContext=@n1.ExecutePin)
   RerouteNode_6 = RigVMDispatchNode(Values=((Type=Control,Name="dyn_Hair_l_base_ctrl"),((),Type=Control,Name="dyn_Hair_l_01_ctrl"),((),Type=Control,Name="dyn_Hair_l_02_ctrl"),((),Type=Control,Name="dyn_Hair_l_03_ctrl"),((),Type=Control,Name="dyn_Hair_l_04_ctrl"),((),Type=Control,Name="None"))) {template="DISPATCH_RigVMDispatch_ArrayMake(in Values,out Array)"}
   RigVMVariableNode(Value=@RerouteNode_6.Array, ExecuteContext=@SequenceAggregate.B) {variable=dyn_Hair_L_ctrls}
   > RigVMVariableNode(Value=/Script/ControlRig.RigUnit_HierarchyGetPoseItemArray(ElementType=All, Initial=false, ItemsToGet=@RerouteNode_6.Array) {is=struct RigVMUnitNode}.Pose) {variable=dyn_Hair_L_pose}
   RerouteNode_6_1 = RigVMDispatchNode(Values="((Type=Control,Name=\"dyn_Hair_Back_01_ctrl\"),((),Type=Control,Name=\"dyn_Hair_Back_02_ctrl\"),((),Type=Control,Name=\"dyn_Hair_Back_03_ctrl\"),((),Type=Control,Name=\"dyn_Hair_Back_04_ctrl\"),((),Type=Control,Name=\"dyn_Hair_Back_05_ctrl\"),((),Type=Control,Name=\" ...(truncated)") {template="DISPATCH_RigVMDispatch_ArrayMake(in Values,out Array)"}
   RigVMVariableNode(Value=@RerouteNode_6_1.Array, ExecuteContext=@SequenceAggregate.A) {variable=dyn_Hair_Back_ctrls}
   > RigVMVariableNode(Value=&RigUnit_HierarchyGetPoseItemArray(ElementType=All, Initial=false, ItemsToGet=@RerouteNode_6_1.Array).Pose) {variable=dyn_Hair_Back_pose}
   RerouteNode_6_2 = RigVMDispatchNode(Values=((Type=Control,Name="dyn_Hair_r_base_ctrl"),((),Type=Control,Name="dyn_Hair_r_01_ctrl"),((),Type=Control,Name="dyn_Hair_r_02_ctrl"),((),Type=Control,Name="dyn_Hair_r_03_ctrl"),((),Type=Control,Name="dyn_Hair_r_04_ctrl"),((),Type=Control,Name="None"))) {template="DISPATCH_RigVMDispatch_ArrayMake(in Values,out Array)"}
   RigVMVariableNode(Value=@RerouteNode_6_2.Array, ExecuteContext=@SequenceAggregate.C) {variable=dyn_Hair_R_ctrls}
   > RigVMVariableNode(Value=&RigUnit_HierarchyGetPoseItemArray(ElementType=All, Initial=false, ItemsToGet=@RerouteNode_6_2.Array).Pose) {variable=dyn_Hair_R_pose}
  ]}
]
hierarchy [
 root {loc=<0 0 0>, rot=<0 0 0 1>}
 pelvis {parent=root, loc=<0 -0.5301 98.0571>, rot=<0.0127 0.707 -0.0127 -0.707>}
 spine_01 {parent=pelvis, loc=<6.1208 -0.9328 0>, rot=<0 0 0.034 -0.9994>}
 spine_02 {parent=spine_01, loc=<6.5996 -0.2095 0>, rot=<0 0 0.0238 -0.9997>}
 spine_03 {parent=spine_02, loc=<6.4914 -0.1237 0>, rot=<0 0 0.0149 -0.9999>}
 spine_04 {parent=spine_03, loc=<6.9598 0.0794 0>, rot=<0 0 -0.0132 -0.9999>}
 spine_05 {parent=spine_04, loc=<6.9987 0 0>, rot=<0 0 -0.0015 -1>}
 neck_01 {parent=spine_05, loc=<11.3138 0.1529 0>, rot=<0 0 -0.2003 -0.9797>}
 neck_02 {parent=neck_01, loc=<4.5627 0 0>, rot=<0 0 0.0504 -0.9987>}
 head {parent=neck_02, loc=<5.3914 0 0>, rot=<0 0 0.115 -0.9934>}
 dyn_Hair_Back_01 {parent=head, loc=<5.8581 -4.633 -0.3258>, rot=<0.1309 -0.9914 0 0>}
 dyn_Hair_Back_02 {parent=dyn_Hair_Back_01, loc=<5.7827 -2.4684 0>, rot=<0 0 -0.004 -1>}
 dyn_Hair_Back_03 {parent=dyn_Hair_Back_02, loc=<6.7665 -0.3183 -0.3554>, rot=<0 0 -0.0413 -0.9991>}
 dyn_Hair_Back_04 {parent=dyn_Hair_Back_03, loc=<9.2159 -0.1146 0>, rot=<0 0 -0.0169 -0.9999>}
 dyn_Hair_Back_05 {parent=dyn_Hair_Back_04, loc=<6.1808 -0.1669 0>, rot=<0 0 -0.023 -0.9997>}
 dyn_Hair_l_base {parent=neck_02, loc=<-2.0549 3.5722 -7.5402>, rot=<-0.4501 -0.893 0 0>}
 dyn_Hair_l_01 {parent=dyn_Hair_l_base, loc=<2.9421 0.2297 -0.0995>, rot=<0 0 0.0317 -0.9995>}
 dyn_Hair_l_02 {parent=dyn_Hair_l_01, loc=<4.4704 0.3718 0.1129>, rot=<0 0 0.0959 -0.9954>}
 dyn_Hair_l_03 {parent=dyn_Hair_l_02, loc=<5.7253 0.1366 0.5179>, rot=<0 0 0.0853 -0.9964>}
 dyn_Hair_l_04 {parent=dyn_Hair_l_03, loc=<5.6891 0.1389 1.697>, rot=<0 0 0 1>}
 dyn_Hair_r_base {parent=neck_02, loc=<-2.0549 3.5722 7.8251>, rot=<-0.4501 -0.893 0 0>}
 dyn_Hair_r_01 {parent=dyn_Hair_r_base, loc=<2.9421 0.2297 -0.0995>, rot=<0 0 0.0317 -0.9995>}
 dyn_Hair_r_02 {parent=dyn_Hair_r_01, loc=<4.4704 0.3718 0.532>, rot=<0 0 0.0959 -0.9954>}
 dyn_Hair_r_03 {parent=dyn_Hair_r_02, loc=<5.7253 0.1366 0.0525>, rot=<0 0 0.0853 -0.9964>}
 dyn_Hair_r_04 {parent=dyn_Hair_r_03, loc=<5.6891 0.1389 -1.0305>, rot=<0 0 0 1>}
 dyn_Hair_l_baseFix {parent=neck_02, loc=<0.5399 -1.9006 -10.542>, rot=<-0.4501 -0.893 0 0>}
 dyn_Hair_r_baseFix {parent=neck_02, loc=<0.3205 -1.851 10.4285>, rot=<-0.4501 -0.893 0 0>}
 clavicle_l {parent=spine_05, loc=<5.4831 -0.1053 -3.2946>, rot=<0.0465 -0.7289 -0.009 -0.6829>}
 upperarm_l {parent=clavicle_l, loc=<9.8658 -1.2136 -0.4105>, rot=<0.0164 -0.3245 -0.0269 -0.9454>}
 lowerarm_l {parent=upperarm_l, loc=<26.465 -0.8345 0.259>, rot=<0.0036 0.0161 -0.2298 -0.9731>}
 hand_l {parent=lowerarm_l, loc=<22.7355 -0.3203 -0.6105>, rot=<-0.6784 -0.0223 0.0735 -0.7307>}
 lowerarm_twist_02_l {parent=lowerarm_l, loc=<10.6707 0 0>, rot=<0 0 0 1>}
 lowerarm_twist_01_l {parent=lowerarm_l, loc=<18.316 0 0>, rot=<0 0 0 1>}
 upperarm_twist_01_l {parent=upperarm_l, loc=<7.2135 -0.1134 0>, rot=<0 0 0 1>}
 upperarm_twist_02_l {parent=upperarm_l, loc=<19.7998 -0.1134 0>, rot=<0 0 0 1>}
 clavicle_r {parent=spine_05, loc=<5.4831 -0.1053 3.2946>, rot=<0.7289 0.0465 -0.6829 0.009>}
 upperarm_r {parent=clavicle_r, loc=<-9.8658 1.2136 0.4105>, rot=<0.0164 -0.3245 -0.0269 -0.9454>}
 lowerarm_r {parent=upperarm_r, loc=<-26.465 0.8345 -0.259>, rot=<0.0036 0.0161 -0.2298 -0.9731>}
 hand_r {parent=lowerarm_r, loc=<-22.7355 0.3203 0.6106>, rot=<-0.6784 -0.0223 0.0735 -0.7307>}
 lowerarm_twist_02_r {parent=lowerarm_r, loc=<-10.6707 0 0>, rot=<0 0 0 -1>}
 lowerarm_twist_01_r {parent=lowerarm_r, loc=<-18.316 0 0>, rot=<0 0 0 -1>}
 upperarm_twist_01_r {parent=upperarm_r, loc=<-7.2135 0.1134 0>, rot=<0 0 0 -1>}
 upperarm_twist_02_r {parent=upperarm_r, loc=<-19.7998 0.1134 0>, rot=<0 0 0 -1>}
 dyn_Hair_Back_01_Null {parent=pelvis, type=Null, loc=<59.5492 -7.9768 -0.3258>, rot=<0.109 -0.994 0 0>}
 dyn_Hair_Back_01_ctrl {controlType=EULER_TRANSFORM, parent=dyn_Hair_Back_01_Null, type=Control, loc=<0 0 0>, rot=<0 0 0 1>}
 dyn_Hair_Back_02_Null {parent=pelvis, type=Null, loc=<54.439 -11.6398 -0.3258>, rot=<-0.1051 0.9945 0 0>}
 dyn_Hair_Back_02_ctrl {controlType=EULER_TRANSFORM, parent=dyn_Hair_Back_02_Null, type=Control, loc=<0 0 0>, rot=<0 0 0 1>}
 dyn_Hair_Back_03_Null {parent=pelvis, type=Null, loc=<47.8884 -13.3652 0.0296>, rot=<0.0639 -0.998 0 0>}
 dyn_Hair_Back_03_ctrl {controlType=EULER_TRANSFORM, parent=dyn_Hair_Back_03_Null, type=Control, loc=<0 0 0>, rot=<0 0 0 1>}
 dyn_Hair_Back_04_Null {parent=pelvis, type=Null, loc=<38.7623 -14.6538 0.0296>, rot=<-0.047 0.9989 0 0>}
 dyn_Hair_Back_04_ctrl {controlType=EULER_TRANSFORM, parent=dyn_Hair_Back_04_Null, type=Control, loc=<0 0 0>, rot=<0 0 0 1>}
 dyn_Hair_Back_05_Null {parent=pelvis, type=Null, loc=<32.6246 -15.4006 0.0296>, rot=<0.0241 -0.9997 0 0>}
 dyn_Hair_Back_05_ctrl {controlType=EULER_TRANSFORM, parent=dyn_Hair_Back_05_Null, type=Control, loc=<0 0 0>, rot=<0 0 0 1>}
 dyn_Hair_r_base_Null {parent=pelvis, type=Null, loc=<45.9219 -0.9591 7.8251>, rot=<-0.109 0.994 0 0>}
 dyn_Hair_r_base_ctrl {controlType=EULER_TRANSFORM, parent=dyn_Hair_r_base_Null, type=Control, loc=<0 0 0>, rot=<0 0 0 1>}
 dyn_Hair_r_01_Null {parent=pelvis, type=Null, loc=<43.9203 1.2094 7.9245>, rot=<0.109 -0.994 0 0>}
 dyn_Hair_r_01_ctrl {controlType=EULER_TRANSFORM, parent=dyn_Hair_r_01_Null, type=Control, loc=<0 0 0>, rot=<0 0 0 1>}
 dyn_Hair_r_02_Null {parent=pelvis, type=Null, loc=<40.6909 4.3229 7.3925>, rot=<-0.109 0.994 0 0>}
 dyn_Hair_r_02_ctrl {controlType=EULER_TRANSFORM, parent=dyn_Hair_r_02_Null, type=Control, loc=<0 0 0>, rot=<0 0 0 1>}
 dyn_Hair_r_03_Null {parent=pelvis, type=Null, loc=<35.7091 7.1479 7.34>, rot=<0.109 -0.994 0 0>}
 dyn_Hair_r_03_ctrl {controlType=EULER_TRANSFORM, parent=dyn_Hair_r_03_Null, type=Control, loc=<0 0 0>, rot=<0 0 0 1>}
 dyn_Hair_r_04_Null {parent=pelvis, type=Null, loc=<30.3549 9.0759 8.3705>, rot=<-0.109 0.994 0 0>}
 dyn_Hair_r_04_ctrl {controlType=EULER_TRANSFORM, parent=dyn_Hair_r_04_Null, type=Control, loc=<0 0 0>, rot=<0 0 0 1>}
 dyn_Hair_l_base_Null {parent=pelvis, type=Null, loc=<45.9219 -0.9591 -7.5402>, rot=<-0.109 0.994 0 0>}
 dyn_Hair_l_base_ctrl {controlType=EULER_TRANSFORM, parent=dyn_Hair_l_base_Null, type=Control, loc=<0 0 0>, rot=<0 0 0 1>}
 dyn_Hair_l_01_Null {parent=pelvis, type=Null, loc=<43.9203 1.2094 -7.4408>, rot=<0.109 -0.994 0 0>}
 dyn_Hair_l_01_ctrl {controlType=EULER_TRANSFORM, parent=dyn_Hair_l_01_Null, type=Control, loc=<0 0 0>, rot=<0 0 0 1>}
 dyn_Hair_l_02_Null {parent=pelvis, type=Null, loc=<40.6909 4.3229 -7.5536>, rot=<-0.109 0.994 0 0>}
 dyn_Hair_l_02_ctrl {controlType=EULER_TRANSFORM, parent=dyn_Hair_l_02_Null, type=Control, loc=<0 0 0>, rot=<0 0 0 1>}
 dyn_Hair_l_03_Null {parent=pelvis, type=Null, loc=<35.7091 7.1479 -8.0715>, rot=<0.109 -0.994 0 0>}
 dyn_Hair_l_03_ctrl {controlType=EULER_TRANSFORM, parent=dyn_Hair_l_03_Null, type=Control, loc=<0 0 0>, rot=<0 0 0 1>}
 dyn_Hair_l_04_Null {parent=pelvis, type=Null, loc=<30.3549 9.0759 -9.7686>, rot=<-0.109 0.994 0 0>}
 dyn_Hair_l_04_ctrl {controlType=EULER_TRANSFORM, parent=dyn_Hair_l_04_Null, type=Control, loc=<0 0 0>, rot=<0 0 0 1>}
 gravity {parent=root, loc=<0 0 0>, rot=<0 0 0 1>}
]
variables [
 dyn_Hair_Back_ctrls {
  type="TArray<FRigElementKey>"
  default=((Type=Control,Name="dyn_Hair_BackL_02"),(Type=Control,Name="dyn_Hair_BackL_03"),(Type=Control,Name="dyn_Hair_BackL_04"),(Type=Control,Name="dyn_Hair_BackL_05"),(Type=Control,Name="dyn_Hair_BackL_06"))
 }
 dyn_Hair_Back_pose {type=FRigPose, default=(Elements=,HierarchyTopologyVersion=-1,PoseHash=-1)}
 dyn_Hair_L_ctrls {
  type="TArray<FRigElementKey>"
  default=((Type=Control,Name="dyn_Hair_BackL_02"),(Type=Control,Name="dyn_Hair_BackL_03"),(Type=Control,Name="dyn_Hair_BackL_04"),(Type=Control,Name="dyn_Hair_BackL_05"),(Type=Control,Name="dyn_Hair_BackL_06"))
 }
 dyn_Hair_L_pose {type=FRigPose, default=(Elements=,HierarchyTopologyVersion=-1,PoseHash=-1)}
 dyn_Hair_R_ctrls {
  type="TArray<FRigElementKey>"
  default=((Type=Control,Name="dfrm_arm_spline_01_r_ctrl"),(Type=Control,Name="dfrm_arm_spline_03_r_ctrl"),(Type=Control,Name="dfrm_arm_spline_05_r_ctrl"),(Type=Control,Name="dfrm_arm_spline_07_r_ctrl"),(Type=Control,Name="dfrm_arm_spline_09_r_ctrl"))
 }
 dyn_Hair_R_pose {type=FRigPose, default=(Elements=,HierarchyTopologyVersion=-1,PoseHash=-1)}
]
