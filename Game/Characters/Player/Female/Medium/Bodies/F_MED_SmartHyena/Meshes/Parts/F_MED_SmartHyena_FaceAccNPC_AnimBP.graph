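# Animation blueprint for the SmartHyena female NPC face accessory, posing hair bones on top of the copied parent-mesh pose.
# A Control Rig, bone constraints and additive offsets build a cached hair pose, with right-side adjustments for shotguns and assault rifles outside emotes.
# "Dyn" caches blend by backbling, weapon and DBNO state when AnimDynamics is enabled, but contain no dynamics nodes.
# The event graph also computes a gravity override, unused by the AnimGraph.
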AnimBlueprint /Game/Characters/Player/Female/Medium/Bodies/F_MED_SmartHyena/Meshes/Parts/F_MED_SmartHyena_FaceAccNPC_AnimBP
target_skeleton=&/F_MED_SmartHyena_FaceAcc_Skeleton
graphs [
 AnimGraph {nodes [
   n1 = "Transform (Modify) Bone - Bone: dyn_Hair_l_02"(ComponentPose=Local To Component(LocalPose=Use cached pose 'HairCachePose'()), Alpha=1, Translation="0,0,0", Rotation="0,8,0", Scale="1,1,1") {
    rotationMode=BMM_Additive
    rotationSpace=BCS_BoneSpace
    scaleMode=BMM_Additive
    scaleSpace=BCS_BoneSpace
    translationMode=BMM_Additive
    translationSpace=BCS_BoneSpace
    alphaBoolBlend {blendOption=Linear}
    boneToModify {boneName=dyn_Hair_l_02}
    rotation {yaw=8}
   }
   n2 = "Transform (Modify) Bone - Bone: dyn_Hair_r_02"(ComponentPose=@n1, Alpha=1, Translation="0,0,0", Rotation="0,8,0", Scale="1,1,1") {
    rotationMode=BMM_Additive
    rotationSpace=BCS_BoneSpace
    scaleMode=BMM_Additive
    scaleSpace=BCS_BoneSpace
    translationMode=BMM_Additive
    translationSpace=BCS_BoneSpace
    alphaBoolBlend {blendOption=Linear}
    boneToModify {boneName=dyn_Hair_r_02}
    rotation {yaw=8}
   }
   n3 = Use cached pose 'HairDynPose'()
   n4 = Use cached pose 'HairCachePose'()
   n5 = SelectFloat(A=0.85, B=0.2, bPickA=Get bIsDBNO())
   Output Pose(Result=Blend Poses by bool(
     BlendPose_0=Two Way Blend(A=@n3, B=@n4, Alpha=@n5) {blendNode {alphaBoolBlend {blendOption=Linear}}}
     BlendPose_1=Component To Local(ComponentPose=@n2)
     bActiveValue=Get bAnimDynamicsEnabled() -> (bAnimDynamicsEnabled:bool)
    ) {blendType=Linear}) {layerGroup=None}
   n6 = Control Rig(Source=Copy Pose From Mesh(SourceMeshComponent=Get Skel_Mesh()) {bUseAttachedParent=true}) {
    alphaBoolBlend {blendOption=Linear}
    controlRigAssetReference {blueprintRigClass=&/F_MED_SmartHyena_FaceAcc_CtrlRig.F_MED_SmartHyena_FaceAcc_CtrlRig_C}
   }
   n7 = "Constraint - dyn_Hair_l_baseFix (PP)"!(ComponentPose=Local To Component(LocalPose=@n6), Alpha=1, ConstraintWeights_0=1, ConstraintWeights_1:float=0) -> (Pose) {
    alphaBoolBlend {blendOption=Linear}
    boneToModify {boneName=dyn_Hair_l_baseFix}
    constraintSetup [{targetBone {boneName=head}}, {targetBone {boneName=dyn_Hair_l_base}}]
    constraintWeights [1,1]
   }
   n8 = "Constraint - dyn_Hair_r_baseFix (PP)"!(ComponentPose=@n7, Alpha=1, ConstraintWeights_0=1, ConstraintWeights_1:float=0) -> (Pose) {
    alphaBoolBlend {blendOption=Linear}
    boneToModify {boneName=dyn_Hair_r_baseFix}
    constraintSetup [{targetBone {boneName=head}}, {targetBone {boneName=dyn_Hair_r_base}}]
    constraintWeights [1,1]
   }
   n9 = "Constraint - dyn_Hair_Back_03 (PPP)"!(ComponentPose=@n8, Alpha=1, ConstraintWeights_0=0.8, ConstraintWeights_1:float=0.1, ConstraintWeights_2:float=0.1) -> (Pose) {
    alphaBoolBlend {blendOption=Linear}
    boneToModify {boneName=dyn_Hair_Back_03}
    constraintSetup [{targetBone {boneName=dyn_Hair_Back_03}}, {targetBone {boneName=clavicle_l}}, {targetBone {boneName=clavicle_r}}]
    constraintWeights [1,1,1]
   }
   n10 = "Transform (Modify) Bone - Bone: dyn_Hair_Back_02"(ComponentPose=@n9, Alpha=1, Translation="0,0,0", Rotation="0,0,0", Scale="1,1,1") {
    rotationMode=BMM_Additive
    rotationSpace=BCS_BoneSpace
    scaleMode=BMM_Additive
    scaleSpace=BCS_BoneSpace
    translationMode=BMM_Additive
    translationSpace=BCS_BoneSpace
    alphaBoolBlend {blendOption=Linear}
    boneToModify {boneName=dyn_Hair_Back_02}
   }
   n11 = "Transform (Modify) Bone - Bone: dyn_Hair_Back_03"(ComponentPose=@n10, Alpha=1, Translation="0,0,0", Rotation="0,3,0", Scale="1,1,1") {
    rotationMode=BMM_Additive
    rotationSpace=BCS_BoneSpace
    scaleMode=BMM_Additive
    scaleSpace=BCS_BoneSpace
    translationMode=BMM_Additive
    translationSpace=BCS_BoneSpace
    alphaBoolBlend {blendOption=Linear}
    boneToModify {boneName=dyn_Hair_Back_03}
    rotation {yaw=3}
   }
   n12 = "Transform (Modify) Bone - Bone: dyn_Hair_r_02"(ComponentPose=@n11, Alpha=1, Translation="0,0,0", Rotation="0,5,0", Scale="1,0.8,1") {
    rotationMode=BMM_Additive
    rotationSpace=BCS_BoneSpace
    scaleMode=BMM_Additive
    scaleSpace=BCS_BoneSpace
    translationMode=BMM_Additive
    translationSpace=BCS_BoneSpace
    alphaBoolBlend {blendOption=Linear}
    boneToModify {boneName=dyn_Hair_r_02}
    rotation {yaw=5}
    scale {y=0.8}
   }
   n13 = Get offset_weapon() -> (offset_weapon:double)
   n14 = "Transform (Modify) Bone - Bone: dyn_Hair_r_base"(ComponentPose=@n12, Alpha=@n13, Translation="0,0,1", Rotation="0,0,5", Scale="1,1,1") {
    rotationMode=BMM_Additive
    rotationSpace=BCS_BoneSpace
    scaleMode=BMM_Additive
    scaleSpace=BCS_BoneSpace
    translationMode=BMM_Additive
    translationSpace=BCS_BoneSpace
    alphaBoolBlend {blendOption=Linear}
    boneToModify {boneName=dyn_Hair_r_base}
    rotation {roll=5}
    translation {z=1}
   }
   n15 = "Transform (Modify) Bone - Bone: dyn_Hair_r_01"(ComponentPose=@n14, Alpha=@n13, Translation="0,-1,3", Rotation="0,10,0", Scale="1,0.7,1") {
    rotationMode=BMM_Additive
    rotationSpace=BCS_BoneSpace
    scaleMode=BMM_Additive
    scaleSpace=BCS_BoneSpace
    translationMode=BMM_Additive
    translationSpace=BCS_BoneSpace
    alphaBoolBlend {blendOption=Linear}
    boneToModify {boneName=dyn_Hair_r_01}
    rotation {yaw=10}
    scale {y=0.7}
    translation {y=-1, z=3}
   }
   n16 = "Transform (Modify) Bone - Bone: dyn_Hair_r_02"(ComponentPose=@n15, Alpha=@n13, Translation="0,0,0", Rotation="10,0,0", Scale="1,1,1") {
    rotationMode=BMM_Additive
    rotationSpace=BCS_BoneSpace
    scaleMode=BMM_Additive
    scaleSpace=BCS_BoneSpace
    translationMode=BMM_Additive
    translationSpace=BCS_BoneSpace
    alphaBoolBlend {blendOption=Linear}
    boneToModify {boneName=dyn_Hair_r_02}
    rotation {pitch=10}
   }
   n17 = "Transform (Modify) Bone - Bone: dyn_Hair_r_03"(ComponentPose=@n16, Alpha=@n13, Translation="0,0,0", Rotation="5,0,0", Scale="1,1,1") {
    rotationMode=BMM_Additive
    rotationSpace=BCS_BoneSpace
    scaleMode=BMM_Additive
    scaleSpace=BCS_BoneSpace
    translationMode=BMM_Additive
    translationSpace=BCS_BoneSpace
    alphaBoolBlend {blendOption=Linear}
    boneToModify {boneName=dyn_Hair_r_03}
    rotation {pitch=5}
   }
   n18 = "Transform (Modify) Bone - Bone: dyn_Hair_r_04"(ComponentPose=@n17, Alpha=@n13, Translation="0,0,0", Rotation="5,0,0", Scale="1,1,1") {
    rotationMode=BMM_Additive
    rotationSpace=BCS_BoneSpace
    scaleMode=BMM_Additive
    scaleSpace=BCS_BoneSpace
    translationMode=BMM_Additive
    translationSpace=BCS_BoneSpace
    alphaBoolBlend {blendOption=Linear}
    boneToModify {boneName=dyn_Hair_r_04}
    rotation {pitch=5}
   }
   n19 = "Constraint - gravity (TT)"!(ComponentPose=@n18, Alpha=1, ConstraintWeights_0=0.5, ConstraintWeights_1:float=0.5) -> (Pose) {
    alphaBoolBlend {blendOption=Linear}
    boneToModify {boneName=gravity}
    constraintSetup [{transformType=Translation, targetBone {boneName=root}}, {transformType=Translation, targetBone {boneName=head}}]
    constraintWeights [1,1]
   }
   Save cached pose 'HairCachePose'(Pose=Component To Local(ComponentPose=@n19)) {cacheName=HairCachePose}
   n20 = Get isWearingPackBool() -> (isWearingPackBool:bool)
   Save cached pose 'HairBackDynPose01'(Pose=Two Way Blend(A=Use cached pose 'HairBackDynPose'(), B=Use cached pose 'HairCachePose'(), Alpha=SelectFloat(A=0.85, B=0.2, bPickA=@n20)) {blendNode {alphaBoolBlend {blendOption=Linear}}}) {cacheName=HairBackDynPose01}
   Save cached pose 'HairBackDynPose'(Pose=Use cached pose 'HairCachePose'()) {cacheName=HairBackDynPose}
   n21 = Get offset_weapon()
   n22 = "Transform (Modify) Bone - Bone: dyn_Hair_r_base"(
    ComponentPose=Local To Component(LocalPose=Use cached pose 'HairBackDynPose01'())
    Alpha=@n21
    Translation="0,0,0"
    Rotation="0,0,0"
    Scale="1,1,1"
   ) {
    rotationMode=BMM_Additive
    rotationSpace=BCS_BoneSpace
    scaleMode=BMM_Additive
    scaleSpace=BCS_BoneSpace
    translationMode=BMM_Additive
    translationSpace=BCS_BoneSpace
    alphaBoolBlend {blendOption=Linear}
    boneToModify {boneName=dyn_Hair_r_base}
   }
   n23 = "Transform (Modify) Bone - Bone: dyn_Hair_r_01"(ComponentPose=@n22, Alpha=@n21, Translation="0,0,0", Rotation="0,0,0", Scale="1,0.8,1") {
    rotationMode=BMM_Additive
    rotationSpace=BCS_BoneSpace
    scaleMode=BMM_Additive
    scaleSpace=BCS_BoneSpace
    translationMode=BMM_Additive
    translationSpace=BCS_BoneSpace
    alphaBoolBlend {blendOption=Linear}
    boneToModify {boneName=dyn_Hair_r_01}
    scale {y=0.8}
   }
   n24 = "Transform (Modify) Bone - Bone: dyn_Hair_r_02"(ComponentPose=@n23, Alpha=@n21, Translation="0,0,0", Rotation="0,0,0", Scale="1,0.8,1") {
    rotationMode=BMM_Additive
    rotationSpace=BCS_BoneSpace
    scaleMode=BMM_Additive
    scaleSpace=BCS_BoneSpace
    translationMode=BMM_Additive
    translationSpace=BCS_BoneSpace
    alphaBoolBlend {blendOption=Linear}
    boneToModify {boneName=dyn_Hair_r_02}
    scale {y=0.8}
   }
   n25 = "Transform (Modify) Bone - Bone: dyn_Hair_r_03"(ComponentPose=@n24, Alpha=@n21, Translation="0,0,0", Rotation="0,0,0", Scale="1,0.5,1") {
    rotationMode=BMM_Additive
    rotationSpace=BCS_BoneSpace
    scaleMode=BMM_Additive
    scaleSpace=BCS_BoneSpace
    translationMode=BMM_Additive
    translationSpace=BCS_BoneSpace
    alphaBoolBlend {blendOption=Linear}
    boneToModify {boneName=dyn_Hair_r_03}
    scale {y=0.5}
   }
   n26 = "Transform (Modify) Bone - Bone: dyn_Hair_r_04"(ComponentPose=@n25, Alpha=@n21, Translation="0,0,0", Rotation="0,0,0", Scale="1,0.5,1") {
    rotationMode=BMM_Additive
    rotationSpace=BCS_BoneSpace
    scaleMode=BMM_Additive
    scaleSpace=BCS_BoneSpace
    translationMode=BMM_Additive
    translationSpace=BCS_BoneSpace
    alphaBoolBlend {blendOption=Linear}
    boneToModify {boneName=dyn_Hair_r_04}
    scale {y=0.5}
   }
   n27 = Get offset_weapon_0() -> (offset_weapon_0:bool)
   Save cached pose 'HairDynPose'(Pose=Two Way Blend(A=Component To Local(ComponentPose=@n26), B=Use cached pose 'HairBackDynPose01'(), Alpha=SelectFloat(A=0.85, B=0, bPickA=@n27)) {blendNode {alphaBoolBlend {blendOption=Linear}}}) {cacheName=HairDynPose}
  ]}
 EventGraph {nodes [
   Event BlueprintUpdateAnimation()
   > n1 = Sequence!(execute) -> (then_0, then_1, then_2:Exec)
   > n2 = Gravity Override State Function DEPRECATED(CharacterPartAnimInstance=Self-Reference(), GameTimeInSeconds=GetGameTimeInSeconds(), GravityOverrideParams=Get Gravity Override Params())
   > Set Gravity Override Cape(Gravity Override Cape:Vector=@n2) -> (Output_Get:Vector)
   n3 = Cast To FortPlayerAnimInstance(execute=@n1.then_1, Object=GetAnimInstance(self=GetOwnerSkeletalMeshComponent()))
   > Set offset_weapon(offset_weapon:double=0) -> (Output_Get:double)
   > Set offset_weapon_0(offset_weapon_0:bool=false) -> (Output_Get:bool)
   > n4 = Switch on EFortWeaponCoreAnimation!(execute, Selection:EFortWeaponCoreAnimation Enum=Get WeaponCoreAnim(self=@n3)) -> (
    Melee:Exec
    Pistol:Exec
    Shotgun:Exec
    PaperBlueprint:Exec
    Rifle:Exec
    MeleeOneHand:Exec
    MachinePistol:Exec
    RocketLauncher:Exec
    GrenadeLauncher:Exec
    GoingCommando:Exec
    AssaultRifle:Exec
    TacticalShotgun:Exec
    SniperRifle:Exec
    TrapPlacement:Exec
    ShoulderLauncher:Exec
    AbilityDecoTool:Exec
    Crossbow:Exec
    C4:Exec
    RemoteControl:Exec
    DualWield:Exec
    AR_BullPup:Exec
    AR_ForwardGrip:Exec
    MedPackPaddles:Exec
    SMG_P90:Exec
    AR_DrumGun:Exec
    Consumable_Small:Exec
    Consumable_Large:Exec
    Balloon:Exec
    MountedTurret:Exec
    CreativeTool:Exec
    ExplosiveBow:Exec
    AshtonIndigo:Exec
    AshtonChicago:Exec
    MeleeDualWield:Exec
    Unarmed:Exec
   )
   n5 = Sequence!(execute=[@n4.AssaultRifle, @n4.TacticalShotgun]) -> (then_0, then_1)
   n6 = Get bIsPlayingEmote()
   Set offset_weapon(execute=[@n5.then_0, @n4.Shotgun], offset_weapon=SelectFloat(A=0, B=1, bPickA=@n6))
   > n7 = Branch(Condition=@n6)
   > Set offset_weapon_0(offset_weapon_0=false)
   Set offset_weapon_0(execute=@n7.else, offset_weapon_0=true)
   Event OnCharacterCustomizationLoaded()
   n8 = CheckForBackbling()
   > Set isWearingPackBool(isWearingPackBool:bool=OR Boolean!(A=@n8, B=@n8.bIsBackblingACape, C:bool=@n8.bIsBackblingLong) -> (ReturnValue)) -> (Output_Get:bool)
   Event BlueprintInitializeAnimation()
   > Set Skel_Mesh(Skel_Mesh=GetOwnerSkeletalMeshComponent())
   AND Boolean(A=@n8, B=NOT Boolean(A=@n8.bIsBackblingLong))
  ]}
]
variables [
 Anim_Dyn {type=number}
 Gravity Override Cape {type=Vector}
 Gravity Override Params {
  type=GravityOverrideParamsStruct
  default {
   onGround_Standing {windAmplitude=<-10 -150 0>, windFrequency=<-7 -1 -1>, windOffset=<-10 -100 0>}
   skydive_DivingDown {gravityOverride=<-500 0 0>}
   skydive_FloatingDown {gravityOverride=<0 -400 100>}
   skydive_GoingUp {gravityOverride=<0 -300 -200>}
   skydive_Parachuting {gravityOverride=<0 -600 -980>}
  }
 }
 isWearingPackBool {type=boolean}
 offset_weapon {type=number}
 offset_weapon_0 {type=boolean}
 Skel_Mesh {type=SkeletalMeshComponent}
]
